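; Server-side Rift Gate spawn manager that places weighted, pre-shuffled gate classes as storm circles progress, destroying itself if its chance roll or calendar-event check fails.
; At startup it gathers locations via one global EQS query or timer-staggered per-anchor queries, then shuffles and filters them using anchor and AI patrol-point proximity.
; Phase-step events set mode-specific counts and run a delayed safe-zone spawn loop until the match cap or locations run out.
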(event EventBeginPlay
  ;[1 1 1] BeginPlay
  (bind (_gamestate _playlist _playlistcontexttags) (Class|FortAsyncActionCurrentPlaylistReady|PlaylistDataReadyAsync)
    (:OnReady
      (if (Networking|IsServer)
        (Variables|Default|SetPlaylist%20Context%20Tags _playlistcontexttags)
        (if (>= (ScalableFloat|GetValueatLevel (Variables|Parameters|GetRow_SpawnManagerEnabledChance)) (Math|Random|RandomFloat))
          ;;[1 1 1] Check if calendar event is active
          (bind _biseventactive (CalendarEvents|GetCalendarEventInformation (Variables|Assets|GetCalendarEventName)))
          (if (or _biseventactive (ScalableFloat|GetValueasBool (Variables|Parameters|GetRow_IgnoreCalendarEvent)))
            (CallFunction|InitializeDropParams _playlistcontexttags)
            (CallFunction|GenerateRandomClassSpawnOrder)
            (Utilities|FlowControl|Delay 5.0)
            (CallFunction|FindAIPatrolPoints)
            (Development|PrintString "Riftgate Spawnmanger Startup; waiting phase change" (Variables|Default|GetDebugText) (Variables|Default|GetDebugText))
            ;;[1 1 1] Bind to game phase step changed to listen for new storm circles
            (bind _returnvalue (GameplayEventRouter|GetEventRouter "/Script/FortniteGame.FortGameState" self))
            (bind (_asynctaskproxy _payload) (GameplayEventRouter|StartListeningToStatefulEvent _returnvalue "/Script/FortniteGame.FortGamePhaseStepUpdatedEvent")
              (:then
                (Variables|Default|SetGamePhaseStepUpdatedAsyncTask _asynctaskproxy)
                (if (CallFunction|IsAnchorModeEnabled?)
                  (CallFunction|StartAnchoredMultiEQSQuery)
                  (else
                    (CallFunction|StartGlobalEQSQuery))))
              (:OnEventReceived
                (CallFunction|OnGameStepChanged (Utilities|Struct|BreakFortGamePhaseStepUpdatedEvent _payload))))
            (else
              (Actor|DestroyActor)))
          (else
            (Actor|DestroyActor)))))))

(event Custom|OnGameStepChanged
  (params (GamePhaseStep :type EAthenaGamePhaseStep%20Enum))
  ;[1 1 1] On Game Step Changed - on new storm circle, decide whether to do a single "global" EQS (1 query to cover entire world) or multiple targeted "anchored" EQS queries
  (if (or (Utilities|Enum|Equal(Enum) GamePhaseStep "StormHolding") (Variables|Default|GetDebugForceSpawningAtStart))
    _
    (elif (Math|Boolean|ANDBoolean (Utilities|Enum|Equal(Enum) GamePhaseStep "StormShrinking") (or (GameplayTags|HasTag (Variables|Default|GetPlaylist%20Context%20Tags) (Variables|LogicVars|GetTeamRumbleTag)) (GameplayTags|HasTag (Variables|Default|GetPlaylist%20Context%20Tags) (Variables|LogicVars|GetBigTeamTag))) (== (Variables|LogicVars|GetCurrentCircle) 0))
      (Development|PrintString "Starting on initial circle shrink for TR / BT" (Variables|Default|GetDebugText) (Variables|Default|GetDebugText))
      (Utilities|FlowControl|Delay 7.0)))
  (CallFunction|UpdatePerCircleDropParams)
  (Math|Integer|IncrementInt (Variables|LogicVars|GetCurrentCircle)
    (:%20%20
      (if (Math|Float|InRange(Float) (Math|Random|RandomFloatinRange 0.0 1.0) 0.0 (ScalableFloat|GetValueatLevel (Variables|Parameters|GetRow_ChanceToSpawnThisCircle)))
        (Variables|LogicVars|SetSpawnCountThisCircle 0)
        (if (> (Variables|LogicVars|GetNumToSpawnThisCircle) 0)
          (CallFunction|SpawnLoop)
          (else
            (Development|PrintString "Nothing to spawn this circle. skipping" (Variables|Default|GetDebugText) (Variables|Default|GetDebugText))))))))

(event Custom|StartGlobalEQSQuery
  ;[1 1 1] Comment
  (Variables|LogicVars|SetNumEQSWaitingToFinish 1)
  (Development|PrintString "starting unanchored global EQS query" (Variables|Default|GetDebugText) (Variables|Default|GetDebugText))
  (bind _returnvalue (AI|EQS|RunEQSQuery (Variables|Assets|GetGlobal%20EQS%20Query%20Template) self "AllMatching"))
  (AI|EQS|BindEventtoOnQueryFinished _returnvalue (EventDispatchers|CreateEvent :!event Custom|OnEQSQueryFinished)))

(event Custom|StartAnchoredMultiEQSQuery
  (bind _eqsanchorpointclass (Variables|Assets|GetEQSAnchorPointClass))
  ;[1 1 1] MultiEQS Query for Anchors; processes anchors 1 by 1
  (if (Utilities|IsValidClass _eqsanchorpointclass)
    (bind _outactors (Actor|GetAllActorsOfClass _eqsanchorpointclass))
    (bind _output_get (Variables|LogicVars|SetAnchorPoints _outactors))
    (Development|PrintString (Utilities|String|Append "Starting Anchored EQS queries. Anchor Points Found: " (Utilities|String|ToString(Integer) (Utilities|Array|Length _output_get))) (Variables|Default|GetDebugText) (Variables|Default|GetDebugText))
    (Variables|LogicVars|SetNumEQSWaitingToFinish (Utilities|Array|Length (Variables|LogicVars|GetAnchorPoints)))
    (Variables|LogicVars|SetcurrentEQSAnchor 0)
    (bind _returnvalue (Utilities|Time|SetTimerbyEvent (delegate Custom|AnchorEQSQueryLoop) 2.0 true))
    (Variables|LogicVars|SetAnchorEQSloopHandle _returnvalue)
    (else
      (Development|PrintString "Anchored MultiEQS failed to run, trying unanchored" (Variables|Default|GetDebugText) (Variables|Default|GetDebugText))
      (CallFunction|StartGlobalEQSQuery))))

(event Custom|AnchorEQSQueryLoop
  (bind _currenteqsanchor (Variables|LogicVars|GetcurrentEQSAnchor))
  (bind _anchorpoints (Variables|LogicVars|GetAnchorPoints))
  (bind _output (Utilities|Array|Get(acopy) _anchorpoints _currenteqsanchor))
  ;[1 1 1] MultiEQS Query for Anchors; processes anchors 1 by 1
  (Utilities|IsValid _output
    (:Is%20Valid
      (bind _returnvalue (AI|EQS|RunEQSQuery (Variables|Assets|GetAnchor%20Point%20EQS%20Query%20Template) _output "AllMatching"))
      (AI|EQS|BindEventtoOnQueryFinished _returnvalue (EventDispatchers|CreateEvent :!event Custom|OnEQSQueryFinished))
      (bind _result (Math|Integer|IncrementInt (Variables|LogicVars|GetcurrentEQSAnchor))
        (:%20%20
          (if (>= _result (Utilities|Array|Length (Variables|LogicVars|GetAnchorPoints)))
            (Utilities|Time|ClearandInvalidateTimerbyHandle (Variables|LogicVars|GetAnchorEQSloopHandle))))))))

(event Custom|OnEQSQueryFinished
  (params (QueryInstance :type EQS%20Query%20Instance%20Object%20Reference) (QueryStatus :type EEnvQueryStatus%20Enum :default "Processing"))
  (bind _numeqswaitingtofinish (Variables|LogicVars|GetNumEQSWaitingToFinish))
  ;[1 1 1] EQS Query callback to return Locations; APPENDS location as this can be called multiple times if using Anchor Points
  (bind (_resultlocations _returnvalue) (AI|EQS|GetQueryResultsasLocations QueryInstance))
  (Utilities|Array|AppendArray (Variables|LogicVars|GetResult%20Locations) _resultlocations)
  (Math|Integer|DecrementInt _numeqswaitingtofinish
    (:%20%20
      (if (<= _numeqswaitingtofinish 0)
        (Development|PrintString "All EQS Anchors have been processed")
        (Utilities|FlowControl|DoOnce false
          (:Completed
            (Utilities|Array|Shuffle (Variables|LogicVars|GetResult%20Locations))
            (CallFunction|PostEQSLocationFiltering)))
        (else
          (Development|PrintString (Utilities|String|Append "still waiting on ESQ to finish: " (Utilities|String|ToString(Integer) _numeqswaitingtofinish)) (Variables|Default|GetDebugText) (Variables|Default|GetDebugText)))))))

(event Custom|SpawnLoop
  ;[1 1 1] EQS Query callback -> spawn drop pods above locations from query
  (if (< (Variables|LogicVars|GetTotalMatchSpawnCount) (Variables|LogicVars|GetMaxNumToSpawnPerMatch))
    (bind (_validlocation _spawnloc) (CallFunction|GetValidLocation))
    (bind _output_get (Variables|LogicVars|SetSpawn%20Loc _spawnloc))
    (if _validlocation
      (CallFunction|SpawnActoratLocation _output_get)
      (Math|Integer|IncrementInt (Variables|LogicVars|GetSpawnCountThisCircle)
        (:%20%20
          (Math|Integer|IncrementInt (Variables|LogicVars|GetTotalMatchSpawnCount)
            (:%20%20
              (if (>= (Variables|LogicVars|GetTotalMatchSpawnCount) (Variables|LogicVars|GetMaxNumToSpawnPerMatch))
                (Development|PrintString (Utilities|String|Append "MANAGER SHUTDOWN: Reached Max # to spawn for match: " (Utilities|String|ToString(Integer) (Variables|LogicVars|GetMaxNumToSpawnPerMatch))))
                (CallFunction|SpawnManagerShutdown)
                (elif (< (Variables|LogicVars|GetSpawnCountThisCircle) (Variables|LogicVars|GetNumToSpawnThisCircle))
                  (Utilities|FlowControl|Delay 3.0)
                  (CallFunction|SpawnLoop)
                  (else
                    (Development|PrintString (Utilities|String|Append "Ending Spawn Loop for this Circle as Max # reached: " (Utilities|String|ToString(Integer) (Variables|LogicVars|GetNumToSpawnThisCircle)) "           Total # spawned this circle: " (Utilities|String|ToString(Integer) (Variables|LogicVars|GetSpawnCountThisCircle)))))))))))
      (else
        (Development|PrintString (Utilities|String|Append "No valid locations found, shutting down spawn manager" (Utilities|String|ToString(Integer) (Variables|LogicVars|GetNumToSpawnThisCircle)) "           Total # spawned this circle: " (Utilities|String|ToString(Integer) (Variables|LogicVars|GetSpawnCountThisCircle))))
        (CallFunction|SpawnManagerShutdown)))
    (else
      (Development|PrintString (Utilities|String|Append "MANAGER SHUTDOWN: Reached Max # to spawn for match: " (Utilities|String|ToString(Integer) (Variables|LogicVars|GetMaxNumToSpawnPerMatch))))
      (CallFunction|SpawnManagerShutdown))))

(event Custom|SpawnManagerShutdown
  ;[1 1 1] unbind
  (bind _gamephasestepupdatedasynctask (Variables|Default|GetGamePhaseStepUpdatedAsyncTask)
    (:then
      (GameplayEventRouter|StopListeningtoStatefulEvent _gamephasestepupdatedasynctask))
    (:else)))

(fn AsyncSphereOverlapCancel
  (return (Game|IsActorBeingDestroyed)))

(fn AsyncSphereOverlapComplete
  (params (WasCancelled :type Boolean)))

(fn AsyncSphereOverlapHitActor
  (params (HitActor :type Actor%20Object%20Reference))
  ;[0.01658 0.1042 0.04287] Destroy BuildingSMActors and teleport away any nearby actors.
  (bind _asbuilding_static_mesh (Utilities|Casting|CastToBuildingSMActor HitActor)
    (:then
      (if (Variables|/Script/Engine.Actor|GetbCanBeDamaged _asbuilding_static_mesh)
        (Gameplay|Die :self _asbuilding_static_mesh :DamageCauser self :!class /Script/FortniteGame.BuildingActor)
        (else
          (Fortnite|FortShippingLog self (Utilities|String|Append "AsyncTrace_ProcessHitActor: Cannot destroy actor: " (Utilities|GetDisplayName _asbuilding_static_mesh)) true))))
    (:CastFailed
      ;[1 1 1] BGAs
      (bind _asbuilding_gameplay_actor (Utilities|Casting|CastToBuildingGameplayActor HitActor)
        (:then
          (if (Variables|/Script/Engine.Actor|GetbCanBeDamaged _asbuilding_gameplay_actor)
            (Actor|DestroyActor _asbuilding_gameplay_actor)))
        (:CastFailed
          ;[1 1 1] Launch Players
          (bind _asfort_pawn (Utilities|Casting|CastToFortPawn HitActor)
            (:then
              (Development|PrintString "launching player pawn" (Variables|Default|GetDebugText) (Variables|Default|GetDebugText))
              (bind _launchvel (CallFunction|CalculateLaunchVelocity 0))
              (Pawn|Character|LaunchCharacterJump _asfort_pawn (* _launchvel (ScalableFloat|GetValueatLevel (Variables|Parameters|GetRow_SpawnKnockbackPawnImpulse))) false false true true))
            (:CastFailed
              ;[1 1 1] Valet Vehicles
              (bind _asfort_athena_vehicle (Utilities|Casting|CastToFortAthenaVehicle HitActor))
              (Development|PrintString "bouncing vehicle" (Variables|Default|GetDebugText) (Variables|Default|GetDebugText))
              (bind _launchvel_2 (CallFunction|CalculateLaunchVelocity _asfort_athena_vehicle))
              (Audio|BroadcastAbilityImpactatLocation 0 _asfort_athena_vehicle (* _launchvel_2 (ScalableFloat|GetValueatLevel (Variables|Parameters|GetRow_SpawnKnockbackVehicleImpulse))) 777.0)
              (Audio|BroadcastAbilitySpinThrustatLocation 0 _asfort_athena_vehicle (* (Math|Vector|SelectVector (Math|Vector|MirrorVectorbyNormal (Vehicle|GetVehicleRightVector _asfort_athena_vehicle) (Vehicle|GetVehicleRightVector _asfort_athena_vehicle)) (Vehicle|GetVehicleRightVector _asfort_athena_vehicle) (> (Vehicle|GetForwardSpeedKmh _asfort_athena_vehicle) 0.0)) 1.5e+02) 777.0))))))))

(fn CalculateLaunchVelocity
  (params (actor :type Actor%20Object%20Reference))
  (bind _output_get (LocalVariables|SetLaunchVelocity (Math|Vector|ClampVectorSize (Math|Vector|MakeVector (* (.x (Transformation|GetVelocity actor)) 1.0) (* (.y (Transformation|GetVelocity actor)) 1.0) (Math|Float|Clamp(Float) (* 1.0 (* (Math|Vector|VectorLength (Transformation|GetVelocity actor)) 0.5)) 5e+02 1.2e+03)) 1.5e+03 2.5e+03)))
  (if (< (+ (Math|Float|Absolute(Float) (.x _output_get)) (Math|Float|Absolute(Float) (.y _output_get))) 3e+02)
    (Variables|SetLaunchVelocity_X (Math|Vector|MakeVector (Math|Random|RandomFloatinRange -1.1e+03 1.1e+03) (Math|Random|RandomFloatinRange -1.1e+03 1.1e+03) (.z _output_get))))
  (return (LocalVariables|GetLaunchVelocity)))

(fn CheckDistanceToAnchors
  (params (loc :type Vector))
  ;[1 1 1] Checks for proximity to ANY anchor; if its close to any anchor it immediately returns as VALID
  (for _array_element (Variables|LogicVars|GetAnchorPoints)
    (if (< (Math|Vector|Distance(Vector) loc (Transformation|GetActorLocation _array_element)) (ScalableFloat|GetValueatLevel (Variables|Parameters|GetRow_EQS_Anchor_Extents)))
      (return true)))
  (return false))

(fn CheckDistanceToNPCPatrolPoints
  (params (loc :type Vector))
  ;[1 1 1] Checks for proximity to ANY patrol Point; if its close to any it immediately returns as INVALID
  (for _array_element (Variables|LogicVars|GetActorsToAvoid)
    (if (< (Math|Vector|Distance(Vector) (Transformation|GetActorLocation _array_element) loc) (ScalableFloat|GetValueatLevel (Variables|Parameters|GetRow_MinDistFromPatrolPoints)))
      (return false)))
  (return true))

(fn CheckDistanceToPreviousSpawns
  (params (loc :type Vector))
  (for _array_element (Variables|LogicVars|GetPreviousSpawnLocs)
    (if (< (Math|Vector|Distance(Vector) _array_element loc) (ScalableFloat|GetValueatLevel (Variables|Parameters|GetRow_MinDistBetweenRiftGates)))
      (return false)))
  (return true))

(fn FindAIPatrolPoints
  (bind _outactors (Actor|GetAllActorsOfClass "/Script/FortniteAI.FortAthenaPatrolPoint"))
  (bind _output_get (Variables|LogicVars|SetActorsToAvoid _outactors))
  (Development|PrintString (Utilities|String|Append "found patrol points: " (Utilities|String|ToString(Integer) (Utilities|Array|Length _output_get))) (Variables|Default|GetDebugText) (Variables|Default|GetDebugText)))

(fn FindTargetPlayer
  ;[1 1 1] This search is N*N algorithm currently, might not need to be optimized since it happens very infrequently
  (LocalVariables|SetCurrentBestPlayerCount 0)
  (LocalVariables|SetCurrentBestController 0)
  (LocalVariables|SetAllPlayerControllers (Fortnite|GetAllFortPlayerControllers true))
  (for _array_element (LocalVariables|GetAllPlayerControllers)
    (bind _outnumplayersinrange (CallFunction|GetNumPlayersNearPlayer _array_element (Variables|LogicVars|GetPlayerSearchRadius)))
    (if (> _outnumplayersinrange (LocalVariables|GetCurrentBestPlayerCount))
      (LocalVariables|SetCurrentBestController _array_element)
      (LocalVariables|SetCurrentBestPlayerCount _outnumplayersinrange)))
  (return (LocalVariables|GetCurrentBestController)))

(fn GenerateRandomClassSpawnOrder
  (Utilities|Array|Clear (Variables|LogicVars|GetClassSpawnSequence))
  ;[1 1 1] Find the max weight range for all items to spawn from 0 to N and then renormalize into 0-1 range
  (for (_ _array_index) (Variables|Assets|GetClasstoSpawn)
    (LocalVariables|SetTotalSpawnWeightRange (+ (ScalableFloat|GetValueatLevel (Variables|Parameters|GetRow%20SpawnChancePerClass) (Math|Conversions|ToFloat(Integer) _array_index)) (LocalVariables|GetTotalSpawnWeightRange))))
  (for (_array_element _array_index_2) (Variables|Assets|GetClasstoSpawn)
    (bind _output_get (LocalVariables|SetNumberOfSpawnsForThisClass (* (Math|Conversions|ToInteger(Boolean) (Utilities|IsValidClass _array_element)) (Math|Float|Ceil (* (Math|Float|Clamp(Float) (/ (ScalableFloat|GetValueatLevel (Variables|Parameters|GetRow%20SpawnChancePerClass) (Math|Conversions|ToFloat(Integer) _array_index_2)) (LocalVariables|GetTotalSpawnWeightRange))) (Variables|LogicVars|GetMaxNumToSpawnPerMatch))))))
    (Development|PrintString (Utilities|String|Append "Class to Spawn: " (Utilities|GetClassDisplayName _array_element) "     Amount: " (Utilities|String|ToString(Integer) _output_get)) (Variables|Default|GetDebugText) (Variables|Default|GetDebugText))
    (for _ (range (LocalVariables|GetNumberOfSpawnsForThisClass))
      (Utilities|Array|Add (Variables|LogicVars|GetClassSpawnSequence) _array_index_2)))
  (Utilities|Array|Shuffle (Variables|LogicVars|GetClassSpawnSequence)))

(fn GetNumPlayersNearPlayer
  (params (InController :type Fort%20Player%20Controller%20Object%20Reference) (Range :type Float%20%28double-precision%29))
  (bind _controller (LocalVariables|GetController))
  (bind _returnvalue (Pawn|GetControlledPawn _controller))
  (LocalVariables|SetInRangeSquared (* Range Range))
  (LocalVariables|SetController InController)
  (Utilities|IsValid _returnvalue
    (:Is%20Valid
      (LocalVariables|SetInPlayerLoc (Transformation|GetActorLocation _returnvalue))
      (for _array_element (Fortnite|GetAllFortPlayerControllers true)
        (Utilities|IsValid (Pawn|GetControlledPawn _array_element)
          (:Is%20Valid
            (if (<= (Math|Vector|Distance2DSquared(Vector) (LocalVariables|GetInPlayerLoc) (Transformation|GetActorLocation (Pawn|GetControlledPawn _array_element))) (LocalVariables|GetInRangeSquared))
              (Math|Integer|IncrementInt (LocalVariables|GetPlayerCount))))))
      (return (LocalVariables|GetPlayerCount)))))

(fn GetRandomDrops
  (params (Min :type Scalable%20Float%20Structure) (Max :type Scalable%20Float%20Structure))
  (return (Math|Random|RandomIntegerinRange (Math|Float|Truncate (ScalableFloat|GetValueatLevel Min)) (Math|Float|Truncate (ScalableFloat|GetValueatLevel Max)))))

(fn GetValidLocation
  (bind _result_locations (Variables|LogicVars|GetResult%20Locations))
  (bind _returnvalue (Utilities|Array|Length _result_locations))
  (bind _result_locations_2 (Variables|LogicVars|GetResult%20Locations))
  (bind _randomindex (LocalVariables|GetrandomIndex))
  (LocalVariables|SetlocationFound false)
  (if (> _returnvalue 0)
    (LocalVariables|SetrandomIndex (Math|Random|RandomIntegerinRange 0 (- _returnvalue 1)))
    (bind _result (ArrayMacros|SafeArrayGet _result_locations_2 _randomindex)
      (:%20%20
        (LocalVariables|SetPotentialLoc _result)
        (Utilities|Array|RemoveIndex _result_locations_2 _randomindex)
        ;[1 1 1] Check location is in Safe Zone
        (bind _returnvalue_2 (SafeZone|IsLocationinSafeZone (LocalVariables|GetPotentialLoc)))
        (if _returnvalue_2
          ;[1 1 1] Distance between Riftgates Check; prevents riftgates from spawning too close together
          (bind _farfrompreviousspawns (CallFunction|CheckDistancetoPreviousSpawns (LocalVariables|GetPotentialLoc)))
          (if _farfrompreviousspawns
            _
            (else
              _))
          (return true (LocalVariables|GetPotentialLoc))))))
  (Development|PrintString "exhausted all valid locations")
  (return false "0, 0, 0"))

(fn HandleNearbyActorsOnSpawn
  (params (spawnloc :type Vector))
  (Class|AsyncOverlapQueue|AsyncSphereOverlapActors :Location spawnloc :Radius (ScalableFloat|GetValueatLevel (Variables|Parameters|GetRow_SpawnKnockbackDestructionRadius)) :ShouldCancelProcess (EventDispatchers|CreateEvent :!event AsyncSphereOverlapCancel) :ProcessHitActorFunction (EventDispatchers|CreateEvent :!event AsyncSphereOverlapHitActor) :CompletionFunction (EventDispatchers|CreateEvent :!event AsyncSphereOverlapComplete) :TraceChannel "ECC_Visibility"))

(fn InitializeDropParams
  (params (InTags :type Gameplay%20Tag%20Container%20Structure))
  (if (GameplayTags|HasTag InTags (Variables|LogicVars|GetTeamRumbleTag))
    ;[1 1 1] Sets the # to spawn per storm circle (for some modes this is the only circle of the match). IN BR this var updates every circle
    (Variables|LogicVars|SetNumToSpawnThisCircle (Math|Float|Truncate (ScalableFloat|GetValueatLevel (Variables|Parameters|GetRow_NumSpawnPerCircle_TeamRumble))))
    ;[1 1 1] Max # to spawn for the match based on game mode
    (Variables|LogicVars|SetMaxNumToSpawnPerMatch (Math|Float|Truncate (ScalableFloat|GetValueatLevel (Variables|Parameters|GetRow_Match_MaxNumToSpawn_TeamRumble))))
    (elif (GameplayTags|HasTag InTags (Variables|LogicVars|GetBigTeamTag))
      (Variables|LogicVars|SetNumToSpawnThisCircle (Math|Float|Truncate (ScalableFloat|GetValueatLevel (Variables|Parameters|GetRow_NumSpawnPerCircle_BigTeam))))
      (Variables|LogicVars|SetMaxNumToSpawnPerMatch (Math|Float|Truncate (ScalableFloat|GetValueatLevel (Variables|Parameters|GetRow_Match_MaxNumToSpawn_BigTeam))))
      (else
        (Variables|LogicVars|SetMaxNumToSpawnPerMatch (Math|Float|Truncate (ScalableFloat|GetValueatLevel (Variables|Parameters|GetRow_Match_MaxSpawn_BR))))
        (Variables|LogicVars|SetInBR true)))))

(fn isAnchorModeEnabled?
  (return (ScalableFloat|GetValueasBool (Variables|Parameters|GetRow_UseAnchoredEQS))))

(fn PostEQSLocationFiltering
  (Development|PrintString (Utilities|String|Append "Post EQS Location Filtering. Returned Locations: " (Utilities|String|ToString(Integer) (Utilities|Array|Length (Variables|LogicVars|GetResult%20Locations)))) (Variables|Default|GetDebugText) (Variables|Default|GetDebugText))
  (bind (_arrayindex _arrayelement) (Utilities|Array|ReverseforEachLoop (Variables|LogicVars|GetResult%20Locations))
    (:LoopBody
      (bind _closetoanchor (CallFunction|CheckDistancetoAnchors _arrayelement))
      (bind _output_get (LocalVariables|SetValidLocation _closetoanchor))
      (if _output_get
        (bind _farfrompreviousspawns (CallFunction|CheckDistancetoNPCPatrolPoints _arrayelement))
        (bind _output_get_2 (LocalVariables|SetValidLocation _farfrompreviousspawns))
        (if _output_get_2
          _
          (else
            _))
        (else
          _))
      (Utilities|Array|RemoveIndex (Variables|LogicVars|GetResult%20Locations) _arrayindex))
    (:Completed
      (Development|PrintString (Utilities|String|Append "Finished Filtering. Remainng valid locations:" (Utilities|String|ToString(Integer) (Utilities|Array|Length (Variables|LogicVars|GetResult%20Locations)))) (Variables|Default|GetDebugText) (Variables|Default|GetDebugText)))))

(fn RemoveUnsafeAnchors
  ;[1 1 1] Check that anchors are inside the safe zone, otherwise permanently remove them
  (bind (_arrayindex _arrayelement) (Utilities|Array|ReverseforEachLoop (Variables|LogicVars|GetAnchorPoints))
    (:LoopBody
      (bind _returnvalue (SafeZone|IsLocationinSafeZone (Transformation|GetActorLocation _arrayelement)))
      (if (not _returnvalue)
        (Utilities|Array|RemoveIndex (Variables|LogicVars|GetAnchorPoints) _arrayindex)
        (Development|PrintString "anchor is outside safe zone, removing" (Variables|Default|GetDebugText) (Variables|Default|GetDebugText))))))

(fn SelectClassToSpawn
  (bind _classtospawn (Variables|Assets|GetClasstoSpawn))
  (bind _returnvalue (Utilities|Array|Length _classtospawn))
  (bind _returnvalue_2 (- _returnvalue 1))
  (bind _returnvalue_3 (Math|Integer|Min(Integer) _returnvalue_2))
  (bind _returnvalue_4 (Math|Random|RandomIntegerinRange 0 _returnvalue_3))
  (if (< (Variables|LogicVars|GetTotalMatchSpawnCount) (Variables|LogicVars|GetMaxNumToSpawnPerMatch))
    (bind _result (ArrayMacros|SafeArrayGet (Variables|LogicVars|GetClassSpawnSequence) (Variables|LogicVars|GetTotalMatchSpawnCount))
      (:%20%20
        (Development|PrintString (Utilities|String|Append "Selected Spawn Class:  " (Utilities|String|ToString(Integer) _result)))
        (return _result)))
    (else
      (Development|PrintString (Utilities|String|Append "Outside Pre-computed random spawn array bounds, randomly selecting a spawn: " (Utilities|String|ToString(Integer) _returnvalue_4)))
      (return _returnvalue_4))))

(fn SpawnActorAtLocation
  (params (Location :type Vector))
  (bind _newparam (CallFunction|SelectClasstoSpawn))
  (if (>= _newparam 0)
    (bind _result (ArrayMacros|SafeArrayGet (Variables|Assets|GetClasstoSpawn) _newparam)
      (:%20%20
        (if (Utilities|IsValidClass _result)
          (bind _returnvalue (Game|SpawnActorfromClass _result (Math|Transform|MakeTransform Location) "AlwaysSpawn" "SelectDefaultAtRuntime" 0 0))
          (Utilities|Array|Add (Variables|LogicVars|GetPreviousSpawnLocs) Location) ; used for range checking future spawns to not be too close
          (Development|PrintString (Utilities|String|Append "Spawned: " (Utilities|GetDisplayName _returnvalue) "           at location: " (Utilities|String|ToString(Vector) Location)) (Variables|Default|GetDebugText) (Variables|Default|GetDebugText))
          (else
            _))))
    (else
      _))
  (Development|PrintString "Class to spawn is invalid, skipping" (Variables|Default|GetDebugText) (Variables|Default|GetDebugText)))

(fn SpawnFromCheat
  (bind _asfort_player_controller (Utilities|Casting|CastToFortPlayerController (Game|GetPlayerController 0)))
  (bind _returnvalue (Fortnite|GetPlayerAimPoint _asfort_player_controller 1e+04 (Utilities|Array|MakeArray 0)))
  (CallFunction|SpawnActoratLocation _returnvalue))

(fn updateAnchorDebugDisplay
  (bind _eqsanchorpointclass (Variables|Assets|GetEQSAnchorPointClass))
  (if (Utilities|IsValidClass _eqsanchorpointclass)
    (bind _outactors (Actor|GetAllActorsOfClass _eqsanchorpointclass))
    (bind _output_get (Variables|LogicVars|SetAnchorPoints _outactors))
    (for _array_element _output_get
      (Utilities|IsValid (Actor|GetComponentbyClass _array_element "/Script/Engine.StaticMeshComponent")
        (:Is%20Valid
          (if (Math|Boolean|ANDBoolean (CallFunction|IsAnchorModeEnabled?) (Variables|Default|GetDebugShowAnchorVolumes?) (Editor|InEditor))
            (Transformation|SetRelativeScale3D (Actor|GetComponentbyClass _array_element "/Script/Engine.StaticMeshComponent") (Math|Vector|MakeVector (* (ScalableFloat|GetValueatLevel (Variables|Parameters|GetRow_EQS_Anchor_Extents)) 0.02) (* (ScalableFloat|GetValueatLevel (Variables|Parameters|GetRow_EQS_Anchor_Extents)) 0.02) 6e+01))
            (Rendering|SetVisibility (Actor|GetComponentbyClass _array_element "/Script/Engine.StaticMeshComponent") true false)
            (else
              (Transformation|SetRelativeScale3D (Actor|GetComponentbyClass _array_element "/Script/Engine.StaticMeshComponent") (1.0 1.0 1.0))
              (Rendering|SetVisibility (Actor|GetComponentbyClass _array_element "/Script/Engine.StaticMeshComponent") false false))))))))

(fn UpdatePerCircleDropParams
  (if (Variables|LogicVars|GetInBR)
    (if (== (Variables|LogicVars|GetCurrentCircle) 0)
      (bind _out (CallFunction|GetRandomDrops (Variables|Parameters|GetRow_MinSpawn_BR_Circle0) (Variables|Parameters|GetRow_MaxSpawn_BR_Circle0)))
      (Variables|LogicVars|SetNumToSpawnThisCircle _out)
      (elif (== (Variables|LogicVars|GetCurrentCircle) 1)
        (bind _out_2 (CallFunction|GetRandomDrops (Variables|Parameters|GetRow_MinSpawn_BR_Circle1) (Variables|Parameters|GetRow_MaxSpawn_BR_Circle1)))
        (Variables|LogicVars|SetNumToSpawnThisCircle _out_2)
        (elif (== (Variables|LogicVars|GetCurrentCircle) 2)
          (bind _out_3 (CallFunction|GetRandomDrops (Variables|Parameters|GetRow_MinSpawn_BR_Circle2) (Variables|Parameters|GetRow_MaxSpawn_BR_Circle2)))
          (Variables|LogicVars|SetNumToSpawnThisCircle _out_3)
          (elif (== (Variables|LogicVars|GetCurrentCircle) 3)
            (bind _out_4 (CallFunction|GetRandomDrops (Variables|Parameters|GetRow_MinSpawn_BR_Circle3) (Variables|Parameters|GetRow_MaxSpawn_BR_Circle3)))
            (Variables|LogicVars|SetNumToSpawnThisCircle _out_4)
            (else
              (bind _out_5 (CallFunction|GetRandomDrops (Variables|Parameters|GetRow_MinSpawn_BR_Circle4) (Variables|Parameters|GetRow_MaxSpawn_BR_Circle4)))
              (Variables|LogicVars|SetNumToSpawnThisCircle _out_5))))))
    (if (Variables|Default|GetDebugText)
      (Development|PrintString (Utilities|String|Append "NEW CIRCLE:" (Utilities|String|ToString(Integer) (Variables|LogicVars|GetCurrentCircle)) "        Max#ToSpawn: " (Utilities|String|ToString(Integer) (Variables|LogicVars|GetNumToSpawnThisCircle)))))))

(fn ConstructionScript
  (CallFunction|UpdateAnchorDebugDisplay))
