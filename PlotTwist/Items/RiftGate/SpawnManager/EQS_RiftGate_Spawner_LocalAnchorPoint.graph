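# EQS query that finds local anchor points for the Rift Gate spawner around the querier.
# A simple grid, sized by scalable-float providers, projects its points onto world geometry with line traces.
# Raised overlap boxes reject water and destructibles, and a nearby-buildings test filters the points further.
# A valid-surface test checks flatness, physical materials and hit actor classes, accepting landscape and static meshes while excluding building actors and landscape spline meshes.
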EnvQuery /PlotTwist/Items/RiftGate/SpawnManager/EQS_RiftGate_Spawner_LocalAnchorPoint
options [{
  generator {
   name=FortEnvQueryGenerator_SimpleGrid_1
   class=FortEnvQueryGenerator_SimpleGrid
   properties {
    bAutoSortTests=true
    bCanRunAsync=false
    envQueryResultNormalizationOption=Default
    generateAround=&EnvQueryContext_Querier
    navDataOverrideContext=null
    optionName=""
    overridenAgentNameForNavmesh=None
    randomDisplacementRatio=0
    gridSize {
     dataBinding=&self:FortEnvQueryGenerator_SimpleGrid_1.FortAIDataProvider_ScalableFloat_0
     dataField=FloatValue
     defaultValue=500
    }
    projectionData {
     bCanDisableTrace=true
     bCanProjectDown=true
     bCanTraceOnGeometry=true
     bCanTraceOnNavMesh=true
     bOnlyBlockingHits=true
     bTraceComplex=false
     extentX=0
     extentY=10
     extentZ=10
     navigationFilter=null
     postProjectionVerticalOffset=0
     projectDown=4000
     projectUp=4000
     serializedChannel=ECC_GameTraceChannel6
     traceChannel=TraceTypeQuery7
     traceMode=GeometryByChannel
     traceProfileName=None
     traceShape=Line
     versionNum=1
    }
    spaceBetween {
     dataBinding=&self:FortEnvQueryGenerator_SimpleGrid_1.FortAIDataProvider_ScalableFloat_1
     dataField=FloatValue
     defaultValue=100
    }
   }
  }
  tests [
   EnvQueryTest_Overlap_10 {
    class=EnvQueryTest_Overlap
    properties {testComment=Reject Water, overlapData {extentX=300, extentY=300, extentZ=300, overlapChannel=ECC_GameTraceChannel15, shapeOffset=<0 0 200>}}
   }
   FortQueryTest_ValidSurface_5 {
    class=FortQueryTest_ValidSurface
    properties {
     flatSurfaceToleranceZ {
      dataBinding=&self:FortQueryTest_ValidSurface_5.FortAIDataProvider_ScalableFloat_3
      dataField=FloatValue
      defaultValue=10
     }
     invalidHitActorClasses [
      /Script/FortniteGame.BuildingDeco
      &BuildingProp
      /Script/FortniteGame.BuildingGameplayActor
      &BuildingStairs
      &BuildingWall
      /Script/FortniteGame.BuildingActor
      /Script/Landscape.LandscapeSplineMeshesActor
     ]
     radius {
      dataBinding=&self:FortQueryTest_ValidSurface_5.FortAIDataProvider_ScalableFloat_0
      dataField=FloatValue
      defaultValue=250
     }
     surfaceMaterials [/Game/FortressPhysicalMaterials/Asphalt, /Game/FortressPhysicalMaterials/Metal, /Game/Balance/Physics/Materials/PhysData_Ice]
     traceOffsetDown {
      dataBinding=&self:FortQueryTest_ValidSurface_5.FortAIDataProvider_ScalableFloat_2
      dataField=FloatValue
      defaultValue=100
     }
     traceOffsetUp {
      dataBinding=&self:FortQueryTest_ValidSurface_5.FortAIDataProvider_ScalableFloat_1
      dataField=FloatValue
      defaultValue=100
     }
     validHitActorClasses [
      /Script/Landscape.LandscapeStreamingProxy
      /Script/Landscape
      &StaticMeshActor
      /Script/Engine.SplineMeshActor
      /Script/Landscape.ControlPointMeshActor
     ]
    }
   }
   EnvQueryTest_Overlap_11 {
    class=EnvQueryTest_Overlap
    properties {
     testComment=Reject Destructibels
     overlapData {extentX=325, extentY=325, extentZ=325, overlapChannel=ECC_GameTraceChannel11, shapeOffset=<0 0 340>}
    }
   }
   FortQueryTest_HasNearbyBuildings_0 {class=FortQueryTest_HasNearbyBuildings, properties {extentXY=5, extentZ=3, testPurpose=Filter}}
  ]
 }]
